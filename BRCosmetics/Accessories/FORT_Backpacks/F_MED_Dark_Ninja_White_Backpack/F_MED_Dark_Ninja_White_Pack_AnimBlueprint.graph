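# Animation Blueprint for the F_MED_Dark_Ninja_White backpack that adds speed-driven rotation to its arm_l and arm_r bones.
# The AnimGraph takes an input pose and applies additive bone-space pitch to each arm in turn, with a shared alpha mapped from PawnSpeed2D and capped at partial strength.
# The EventGraph toggles Sprinting_Alpha on a speed threshold each update, but the AnimGraph does not read that variable.
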AnimBlueprint /BRCosmetics/Accessories/FORT_Backpacks/F_MED_Dark_Ninja_White_Backpack/F_MED_Dark_Ninja_White_Pack_AnimBlueprint
target_skeleton=&/F_MED_Dark_Ninja_White_Pack_Skeleton
graphs [
 AnimGraph {nodes [
   AnimGraphNode_SubInput_0 = Input Pose()
   n1 = MapRangeClamped(Value=Get PawnSpeed2D(), InRangeA=0, InRangeB=1250, OutRangeB=0.37)
   n2 = "Transform (Modify) Bone - Bone: arm_r"(ComponentPose=Local To Component(LocalPose=@AnimGraphNode_SubInput_0), Alpha=@n1, Translation="0,0,0", Rotation="20,0,0", Scale="1,1,1") {
    rotationMode=BMM_Additive
    rotationSpace=BCS_BoneSpace
    alphaBoolBlend {blendOption=Linear}
    alphaScaleBiasClamp {interpSpeedDecreasing=100, interpSpeedIncreasing=100}
    boneToModify {boneName=arm_r}
    rotation {pitch=20}
   }
   n3 = "Transform (Modify) Bone - Bone: arm_l"(ComponentPose=@n2, Alpha=@n1, Translation="0,0,0", Rotation="20,0,0", Scale="1,1,1") {
    rotationMode=BMM_Additive
    rotationSpace=BCS_BoneSpace
    alphaBoolBlend {blendOption=Linear}
    alphaScaleBiasClamp {interpSpeedDecreasing=100, interpSpeedIncreasing=100}
    boneToModify {boneName=arm_l}
    rotation {pitch=20}
   }
   Output Pose(Result=Component To Local(ComponentPose=@n3)) {layerGroup=None}
  ]}
 EventGraph {nodes [
   Event BlueprintUpdateAnimation()
   > n1 = Branch(Condition="float > float"(A=Get PawnSpeed2D(), B=500) {is=title K2Node_CallFunction})
   > Set Sprinting_Alpha(Sprinting_Alpha:double=1) -> (Output_Get:double)
   Set Sprinting_Alpha(execute=@n1.else, Sprinting_Alpha=0)
   TryGetPawnOwner()
  ]}
]
variables [Sprinting_Alpha {type=number}]
