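# Physics asset for the female medium Dark Ninja White backpack cosmetic, giving its skeletal mesh simple collision.
# It holds a single capsule body on the pack bone, offset and rotated to fit the pack.
# It defines no constraints, so there are no jointed or secondary bodies.
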PhysicsAsset /BRCosmetics/Accessories/FORT_Backpacks/F_MED_Dark_Ninja_White_Backpack/F_MED_Dark_Ninja_White_Pack_PhysicsAsset
bodies {pack [{shapeType=Capsule, length=20.8623, radius=10.866, center=<0.0005 -2.573 -12.4468>, rotation=<rot 0.2366 -90 -0.0043>}]}
constraints []
